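; Control component for the ToyFox secondary player's seeking missile that relays control lifecycle events.
; When control starts, it broadcasts OnStartedControlling with the Bike Kid it receives.
; When dismissal starts, it broadcasts OnDismissed without passing on the dismissal reason tag.
; The graph holds only this forwarding, with no missile movement or seeking logic.
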(event EventOnStartedControlling
  (params (InBikeKid :type Fort%20Bike%20Kid%20Object%20Reference))
  (Default|CallOnStartedControlling InBikeKid))

(event EventOnDismissalStarted
  (params (DismissalReason :type Gameplay%20Tag%20Structure))
  (Default|CallOnDismissed))
