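; Player-piloted ToyFox seeking missile pawn, flown through the BikeKid controlling component and targetable by hostile auto-fire and aim assist.
; On spawn it briefly ignores giant-pawn collision, while the server loosens movement correction, applies a projectile effect to the instigator and starts control.
; On destruction the server releases control and applies pawn and world damage via async targeting; clients trace for water to pick explosion FX.
; Land explosions trigger an explosion camera, cleared on dismissal.
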(event EventBeginPlay
  (bind _returnvalue (Game|GetInstigatorController))
  (bind _returnvalue_2 (Game|GetInstigator))
  (bind _returnvalue_3 (Ability|GetFortAbilitySystemComponent _returnvalue_2))
  ; Begin Play
  ;;[1 1 1] Sequence just for organizational purposes
  (Utilities|FlowControl|Sequence
    (:then_0
      ;; We disable the collision with the owner for s short time after the spawn to avoid possible collision issues during the deploy.
      (CallFunction|ToggleCollisionwithOwningPawn false)
      (Utilities|FlowControl|Delay (ScalableFloat|GetCachedScalableFloatValue (Variables|Default|GetHF_LaunchIgnoreCollisionDuration)))
      (CallFunction|ToggleCollisionwithOwningPawn true))
    (:then_1
      ;;[1 1 1] Var set up & Hotfixes
      (bind _asfort_toy_fox_movement_component (Utilities|Casting|CastToFortToyFoxMovementComponent (Pawn|GetMovementComponent)))
      (Class|FortBikeKidMovementComponent|SetHitBoxComponent _asfort_toy_fox_movement_component (Variables|Default|GetHitBox))
      (Class|FortToyFoxMovementComponent|IgnoreActorforOverlapCheck _asfort_toy_fox_movement_component (Giant|GetGiantPawn (Actor|GetComponentbyClass (Fortnite|Athena|GameState|GetGameStateAthena) "/ToyFoxGameplay/Gameplay/B_ToyFoxGameStateComponent.B_ToyFoxGameStateComponent_C")) (ScalableFloat|GetCachedScalableFloatValue (Variables|Default|GetHF_LaunchIgnoreCollisionDuration))))
    (:then_2
      (Effects|Components|ParticleSystem|SetBoolParameter (Variables|Default|GetNS_RocketTrail) "IsLocalPlayer" (Pawn|IsLocalPlayerController (Game|GetInstigatorController)))
      ;;[0.1168 1 0.08209] Aim Assist
      (AimAssist|RegisterAimAssistTarget 0 (Variables|Character|GetCapsuleComponent)))
    (:then_3
      (Networking|SwitchHasAuthority
        (:Authority
          (Utilities|FlowControl|Sequence
            (:then_0
              ;;[0.1034 0.04759 0.15 0.5] Projectile pawn is moving quickly, prevent error corrections as much as possible for smoothness.
              (bind _asfort_movement_comp_character_athena (Utilities|Casting|CastToFortMovementComp_CharacterAthena (Variables|Character|GetCharacterMovement)))
              (NetTolerance|SetUseHighAccelerationTolerance _asfort_movement_comp_character_athena true "ToyFoxSeeking"))
            (:then_1
              ;;[0.007261 0.1171 0.15 0.5] Apply pawn projectile class when ability system component is ready.
              (Utilities|IsValid _returnvalue_3
                (:Is%20Valid
                  _)
                (:Is%20Not%20Valid
                  (Utilities|FlowControl|DelayUntilNextTick) ; OnAbilitySystemComponentInitialized delegate is not firing for some reason.
                ))
              (bind _returnvalue_4 (GameplayEffects|ApplyGameplayEffectToSelf :self _returnvalue_3 :GameplayEffectClass (Variables|Default|GetPawnProjectileGameplayEffect) :Level 0.0))
              (Variables|Default|SetPawnProjGEHandle _returnvalue_4))
            (:then_2
              (Class|FortBikeKidControllingComponent|StartControllingBikeKid (Actor|GetComponentbyClass (Fortnite|GetFortPlayerControllerfromActor (Game|GetInstigator)) "/Script/BikeKidRuntime.FortBikeKidControllingComponent") self))))
        (:Remote
          (CallFunction|PlayFireSound)
          (Utilities|IsValid _returnvalue
            (:Is%20Valid
              (if (Pawn|IsLocalPlayerController _returnvalue)
                (Components|Activation|Activate (Variables|Default|GetBankingAudio) false)))))))))

(event EventBP_OnDestructionStarted
  (bind _returnvalue (Game|GetInstigator))
  (bind _returnvalue_2 (Ability|GetAbilitySystemComponent _returnvalue))
  (bind _returnvalue_3 (GameplayEffects|MakeEffectContext _returnvalue_2))
  ; Destruction start
  (Utilities|FlowControl|Sequence
    (:then_0
      (Networking|SwitchHasAuthority
        (:Authority
          (Class|FortBikeKidControllingComponent|TrytoStopControllingBikeKid (Actor|GetComponentbyClass (Fortnite|GetFortPlayerControllerfromActor (Game|GetInstigator)) "/Script/BikeKidRuntime.FortBikeKidControllingComponent") (Variables|Default|GetDismissData))
          ; Async apply damages
          (Utilities|IsValid _returnvalue_2
            (:Is%20Valid
              (bind (_asynctaskproxy _targetinghandle) (Class|AsyncActionPerformTargeting|PerformTargetingAsyncAction 0 (Variables|Default|GetExplosionTargetingPResetPawn) true)
                (:then
                  (bind (_asynctaskproxy_2 _targetinghandle_2) (Class|AsyncActionPerformTargeting|PerformTargetingAsyncAction 0 (Variables|Default|GetExplosionTargetingPresetWorld) true)
                    (:Targeted
                      (Ability|Tasks|ApplyGameplayEffectToTargetingHandleAsync _returnvalue_2 _targetinghandle_2 (Variables|Default|GetExplosionWorldDamageGameplayEffect) 1 1 _returnvalue_3 (Variables|Default|GetHotfix_MaxTargetsToProcessPerTick)))))
                (:Targeted
                  (Ability|Tasks|ApplyGameplayEffectToTargetingHandleAsync _returnvalue_2 _targetinghandle (Variables|Default|GetExplosionPawnDamageGameplayEffect) 1 1 _returnvalue_3 (Variables|Default|GetHotfix_MaxTargetsToProcessPerTick)))))))
        (:Remote
          (bind _outhit (Collision|LineTraceByChannel (+ (Transformation|GetActorLocation) "0.000000,0.000000,1000.000000") (+ (Transformation|GetActorLocation) "0.000000,0.000000,-1000.000000") "TraceTypeQuery12"))
          (if (.bSuccess (Utilities|Casting|CastToWaterBody (.HitActor (Collision|BreakHitResult _outhit))))
            (bind _returnvalue_4 (Niagara|SpawnSystematLocation :SystemTemplate (Variables|Default|GetWaterExplosionFX) :Location (.ImpactPoint (Collision|BreakHitResult _outhit)) :Scale "1.000000,1.000000,1.000000" :bAutoDestroy true :bAutoActivate true :PoolingMethod "AutoRelease" :bPreCullCheck true))
            (Effects|Components|ParticleSystem|SetFloatParameter _returnvalue_4 "EffectScale" 2.0)
            (bind _returnvalue_5 (Niagara|SpawnSystematLocation "/ToyFoxGameplay/Assets/FX/Cannon/NS_ToyFox_Cannon_Explosion.NS_ToyFox_Cannon_Explosion" (.ImpactPoint (Collision|BreakHitResult _outhit)) "90.000000,0.000000,0.000000" "1.000000,1.000000,1.000000" true true "AutoRelease"))
            (Effects|Components|ParticleSystem|SetFloatParameter _returnvalue_5 "Global Scale" 0.5)
            (Effects|Components|ParticleSystem|SetBoolParameter _returnvalue_5 "IsWater" true)
            (Effects|Components|ParticleSystem|SetBoolParameter _returnvalue_5 "IsCannonExp" false)
            (else
              (bind _returnvalue_6 (Niagara|SpawnSystematLocation :SystemTemplate (Variables|Default|GetNSLandExplosion) :Location (Transformation|GetActorLocation) :Scale "1.000000,1.000000,1.000000" :bAutoDestroy true :bAutoActivate true :PoolingMethod "AutoRelease" :bPreCullCheck true))
              (Effects|Components|ParticleSystem|SetFloatParameter _returnvalue_6 "Global Scale" 1.6)
              (Audio|Components|Audio|FadeOut (Variables|Default|GetSound_Flying_Loop) 0.2 0.0 "Linear" :!class /Script/Engine.AudioComponent)
              (CallFunction|PlayExplosionSound)
              (Rendering|SetActorHiddenInGame true)
              (Development|SetHiddeninGame (Variables|Default|GetNS_RocketTrail) true true)
              (SoundIndicator|StopSoundIndicatorLooping (Variables|Default|GetFortSoundIndicator))
              (if (not (Variables|Default|GetBikeKidDismissed))
                (CallFunction|SetExplodeCamera true)))))))
    (:then_1
      (Collision|SetActorEnableCollision false))))

(event EventBP_OnBikeKidDismissed
  (params (DismissalReason :type Gameplay%20Tag%20Structure%20%28by%20ref%29))
  ; On dismissed (stopped controlling)
  (GameplayEffects|RemoveActiveGameplayEffect (Ability|GetFortAbilitySystemComponent (Game|GetInstigator)) (Variables|Default|GetPawnProjGEHandle) -1)
  (CallFunction|SetExplodeCamera false)
  (Default|CallOnDisconnected self)
  (Variables|Default|SetBikeKidDismissed true))

(event EventBP_OnBikeKidLaunched)

(event EventEndPlay
  (params (EndPlayReason :type EEndPlayReason%20Enum :default "Destroyed"))
  (|CallParentFunction EndPlayReason))

(event Custom|Play%20Fire%20Sound
  (bind _returnvalue (Game|GetInstigatorController))
  ; Projectile fire audio
  (Networking|CanExecuteCosmeticEvents
    (:True
      (Utilities|IsValid _returnvalue
        (:Is%20Valid
          (if (Pawn|IsLocalPlayerController _returnvalue)
            (Audio|PlaySound2D (Variables|Sounds|GetFire%20Sound%201p))
            (else
              _)))
        (:Is%20Not%20Valid
          _))
      (Audio|PlaySoundatLocation (Variables|Sounds|GetFire%20Sound%203p) (Transformation|GetWorldLocation (Variables|Character|GetCapsuleComponent))))))

(event Custom|Play%20Explosion%20Sound
  (bind _explosionsound (Variables|Sounds|GetExplosionSound))
  ; Explosion sound
  (if (Pawn|IsLocallyViewed)
    (Audio|PlaySound2D _explosionsound)
    (else
      (Audio|PlaySoundatLocation _explosionsound (Transformation|GetActorLocation)))))

(event EventOnDeathPlayEffects
  (params (Damage :type Float%20%28single-precision%29) (DamageTags :type Gameplay%20Tag%20Container%20Structure%20%28by%20ref%29) (Momentum :type Vector :default "0, 0, 0") (HitInfo :type Hit%20Result%20Structure%20%28by%20ref%29) (InstigatedBy :type Fort%20Pawn%20Object%20Reference) (DamageCauser :type Actor%20Object%20Reference) (EffectContext :type Gameplay%20Effect%20Context%20Handle%20Structure)))

(event EventOnDamagePlayEffects
  (params (Damage :type Float%20%28single-precision%29) (DamageTags :type Gameplay%20Tag%20Container%20Structure%20%28by%20ref%29) (Momentum :type Vector :default "0, 0, 0") (HitInfo :type Hit%20Result%20Structure%20%28by%20ref%29) (InstigatedBy :type Fort%20Pawn%20Object%20Reference) (DamageCauser :type Actor%20Object%20Reference) (EffectContext :type Gameplay%20Effect%20Context%20Handle%20Structure)))

(fn DebugPrintSpeed
  (Development|PrintString (Utilities|String|Append "Max Fly Speed::  " (Utilities|String|ToString(Float) (Variables|/Script/Engine.CharacterMovementComponent|GetMaxFlySpeed (Variables|Character|GetCharacterMovement)))) true true "(R=0.092793,G=1.000000,B=0.156864,A=1.000000)" 5.0))

(fn GetAffiliationsToShowFor
  (return (Utilities|Array|MakeArray (Utilities|Enum|LiteralenumEFortTeamAffiliation "Hostile"))))

(fn GetTargetOwner
  (return 0))

(fn IsValidAutoFireTarget
  (params (Attacker :type Actor%20Object%20Reference))
  (return (Utilities|Enum|Equal(Enum) (Team|GetTeamAffiliationBetweenTeams (Team|GetActorTeam Attacker) (Team|GetActorTeam self)) "Hostile")))

(fn SetExplodeCamera
  (params (bEnable :type Boolean))
  (bind _cammodifiertouse (LocalVariables|GetCamModifierToUse))
  (if (!= bEnable (Variables|Default|GetExplosionCamEnabled))
    (Variables|Default|SetExplosionCamEnabled bEnable)
    (bind _asfort_player_controller (Utilities|Casting|CastToFortPlayerController (Game|GetInstigatorController)))
    (bind _asfort_player_camera_base (Utilities|Casting|CastToFortPlayerCameraBase (Variables|/Script/Engine.PlayerController|GetPlayerCameraManager _asfort_player_controller)))
    (Utilities|FlowControl|Sequence
      (:then_0
        ; Set up camera mode
        (if (Variables|GetbEnable)
          (Camera|AddCameraModeOverride _asfort_player_camera_base (Variables|Default|GetCameraMode_Explode) (Variables|Default|GetCameraMode_Explode_Tag) 0 0 0)
          (else
            (Camera|RemoveCameraModeOverride _asfort_player_camera_base (Variables|Default|GetCameraMode_Explode_Tag) 0 false))))
      (:then_1
        ; Find or get camera modifier
        (bind _returnvalue (Game|Player|FindCameraModifierbyClass _asfort_player_camera_base "/ToyFoxGameplay/Gameplay/SecondaryPlayer/SeekingRocket/B_ToyFox_SeekMissile_ExplodeCamMod.B_ToyFox_SeekMissile_ExplodeCamMod_C"))
        (Utilities|IsValid _returnvalue
          (:Is%20Valid
            (LocalVariables|SetCamModifierToUse _returnvalue))
          (:Is%20Not%20Valid
            (bind _returnvalue_2 (Game|Player|AddNewCameraModifier _asfort_player_camera_base "/ToyFoxGameplay/Gameplay/SecondaryPlayer/SeekingRocket/B_ToyFox_SeekMissile_ExplodeCamMod.B_ToyFox_SeekMissile_ExplodeCamMod_C"))
            (LocalVariables|SetCamModifierToUse _returnvalue_2))))
      (:then_2
        ; Set up camera modifier
        (if (Variables|GetbEnable)
          (Variables|/ToyFoxGameplay/Gameplay/SecondaryPlayer/SeekingRocket/B_ToyFox_SeekMissile_ExplodeCamMod.B_ToyFox_SeekMissile_ExplodeCamMod_C|SetOriginalLoc (Camera|GetCameraLocation _asfort_player_camera_base) _cammodifiertouse)
          (CameraModifier|EnableModifier _cammodifiertouse)
          (else
            (CameraModifier|DisableModifier (LocalVariables|GetCamModifierToUse) true)))))))

(fn ShouldAlliesAutoFireOnDeployedTarget
  (return false))

(fn ShouldAutoFireOnSelfDeployedTarget
  (return false))

(fn ShouldShowSoundIndicator
  (params (PlayerController :type Fort%20Player%20Controller%20Object%20Reference))
  (return true))

(fn ToggleCollisionWithOwningPawn
  (params (Enabled :type Boolean))
  (bind _output_get (LocalVariables|SetGiantPawn (Giant|GetGiantPawn (Actor|GetComponentbyClass (Fortnite|Athena|GameState|GetGameStateAthena) "/ToyFoxGameplay/Gameplay/B_ToyFoxGameStateComponent.B_ToyFoxGameStateComponent_C"))))
  (Collision|IgnoreActorwhenMoving (Variables|Character|GetCapsuleComponent) _output_get (not (Variables|GetEnabled)))
  (Collision|IgnoreActorwhenMoving (Variables|/Script/Engine.Character|GetCapsuleComponent (LocalVariables|GetGiantPawn)) self (not (Variables|GetEnabled))))

(fn ConstructionScript)
