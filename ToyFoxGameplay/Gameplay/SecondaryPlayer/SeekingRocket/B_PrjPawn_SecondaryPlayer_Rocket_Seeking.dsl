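; Remote-controlled seeking rocket pawn for the ToyFox secondary player that explodes on kill, ignoring Megazord (giant pawn) collision.
; BeginPlay raises movement error tolerance for smoothness and applies the projectile GameplayEffect on authority.
; Vehicle overlap and water trigger destruction; on kill, chained async targeting applies instigator-sourced damage to pawns, then world.
; Replicated flags drive explosion cosmetics and keep physics running after unpossession.
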(event EventBeginPlay
  (bind _self self)
  (bind _returnvalue (Ability|GetFortAbilitySystemComponent _self))
  ;[0 0.15 0.01449 0.5] BeginPlay: Initialize variables and cosmetics.
  (|CallParentFunction)
  ;; Ignore megazord collision for the missile
  (Collision|IgnoreActorwhenMoving (Variables|Character|GetCapsuleComponent) (Giant|GetGiantPawn (Actor|GetComponentbyClass (Fortnite|Athena|GameState|GetGameStateAthena) "/ToyFoxGameplay/Gameplay/B_ToyFoxGameStateComponent.B_ToyFoxGameStateComponent_C")) true)
  (Effects|Components|ParticleSystem|SetBoolParameter (Variables|Default|GetNS_RocketTrail_RC) "IsLocalPlayer" (Pawn|IsLocalPlayerController (Game|GetInstigatorController)))
  (Networking|SwitchHasAuthority
    (:Authority
      (Utilities|FlowControl|Sequence
        (:then_0
          ;[0.1034 0.04759 0.15 0.5] Projectile pawn is moving quickly, prevent error corrections as much as possible for smoothness.
          (bind _asfort_movement_comp_character_athena (Utilities|Casting|CastToFortMovementComp_CharacterAthena (Variables|Character|GetCharacterMovement)))
          (NetTolerance|SetUseHighAccelerationTolerance _asfort_movement_comp_character_athena true "ToyFoxSeeking"))
        (:then_1
          ;[0.007261 0.1171 0.15 0.5] Apply pawn projectile class when ability system component is ready.
          (Utilities|IsValid _returnvalue
            (:Is%20Valid
              _)
            (:Is%20Not%20Valid
              (Utilities|FlowControl|DelayUntilNextTick) ; OnAbilitySystemComponentInitialized delegate is not firing for some reason.
            ))
          (GameplayEffects|ApplyGameplayEffectToSelf :self _returnvalue :GameplayEffectClass (Variables|Settings|GetPawnProjectileGameplayEffect) :Level 0.0))))))

(event RemoteControlledPawn|EventOnDoKill
  (bind _returnvalue (Game|GetInstigator))
  (bind _returnvalue_2 (Ability|GetFortAbilitySystemComponent _returnvalue))
  (bind _returnvalue_3 (GameplayEffects|MakeEffectContext _returnvalue_2))
  (bind _hotfix_maxtargetstoprocesspertick (Variables|Hotfix|GetHotfix_MaxTargetsToProcessPerTick))
  ;[0.15 0.07798 0.1432 0.5] OnDoKill: Perform damage using async target selection.
  (|CallParentFunction)
  (Networking|SwitchHasAuthority
    (:Authority
      (Variables|Transient|SetbKilled true)
      (Utilities|IsValid _returnvalue
        (:Is%20Valid
          (Utilities|IsValid _returnvalue_2
            (:Is%20Valid
              (bind (_asynctaskproxy _targetinghandle) (Class|AsyncActionPerformTargeting|PerformTargetingAsyncAction 0 (Variables|TargetingPresets|GetExplosionTargetingPresetPawn) true)
                (:Targeted
                  (bind _asynctaskproxy_2 (Ability|Tasks|ApplyGameplayEffectToTargetingHandleAsync _returnvalue_2 _targetinghandle (Variables|Settings|GetExplosionPawnDamageGameplayEffect) 1 1 _returnvalue_3 _hotfix_maxtargetstoprocesspertick))
                  (Default|WaitAsyncActionComplete _asynctaskproxy_2
                    (:Complete
                      (bind (_asynctaskproxy_3 _targetinghandle_2) (Class|AsyncActionPerformTargeting|PerformTargetingAsyncAction 0 (Variables|TargetingPresets|GetExplosionTargetingPresetWorld) true)
                        (:Targeted
                          (Ability|Tasks|ApplyGameplayEffectToTargetingHandleAsync _returnvalue_2 _targetinghandle_2 (Variables|Settings|GetExplosionWorldDamageGameplayEffect) 1 1 _returnvalue_3 _hotfix_maxtargetstoprocesspertick))))))))
            (:Is%20Not%20Valid
              (Fortnite|FortShippingLog self "OnDoKill: ASC is null." true))))
        (:Is%20Not%20Valid
          (Fortnite|FortShippingLog self "OnDoKill: Instigator is null." true))))))

(event EventOnDisplayHitNotify
  (params (DamageDealt :type Float%20%28single-precision%29) (bCriticalHit :type Boolean) (bFatalHit :type Boolean) (bShieldHit :type Boolean) (HitActor :type Actor%20Object%20Reference) (HitTags :type Gameplay%20Tag%20Container%20Structure) (HitLocation :type Vector :default "0, 0, 0"))
  ;[0.15 0.02543 0.1465 0.5] Added so damage numbers are displayed correctly for the damage done by this rocket pawn.
  (EventDispatchers|CallDisplayHitNotify DamageDealt bCriticalHit bFatalHit bShieldHit HitActor HitLocation HitTags))

(event EventOnCharacterPartsReinitialized
  ;[0.04941 0.1264 0.15 0.5] Turn off character parts visibility so that no T-pose player standing on rocket
  (CallFunction|SetCharacterBodyVisibilityforPossession false))

(event EventUnpossessed
  (params (OldController :type Controller%20Object%20Reference))
  ;[0.15 0.09739 0.02097 0.5] Unpossessed: On unpossess, allow the prj pawn to run physics with no controller.
  (|CallParentFunction OldController)
  (Networking|SwitchHasAuthority
    (:Authority
      (Variables|Transient|SetbControlledUnpossessed true))))

(event Custom|PlayExplosionEffects
  ;[0 0.1489 0.15 0.5] Play the explosion effects on death.
  (Networking|CanExecuteCosmeticEvents
    (:True
      (bind _returnvalue (Niagara|SpawnSystematLocation :SystemTemplate (Variables|FX|GetNS_LandExplosion) :Location (Transformation|GetActorLocation) :Scale "1.000000,1.000000,1.000000" :bAutoDestroy true :bAutoActivate true :PoolingMethod "AutoRelease" :bPreCullCheck true))
      (Effects|Components|ParticleSystem|SetFloatParameter _returnvalue "Global Scale" 1.6)
      (Audio|Components|Audio|FadeOut (Variables|Default|GetMissileInAirAudio) 0.2 0.0 "Linear" :!class /Script/Engine.AudioComponent)
      (Audio|PlaySoundatLocation (Variables|FX|GetExplosionSound) (Transformation|GetActorLocation))
      (Rendering|SetActorHiddenInGame true)
      (Development|SetHiddeninGame (Variables|Default|GetNS_RocketTrail_RC) true true))))

(event OnComponentBeginOverlap(HitBox)
  (params (OverlappedComponent :type Primitive%20Component%20Object%20Reference) (OtherActor :type Actor%20Object%20Reference) (OtherComp :type Primitive%20Component%20Object%20Reference) (OtherBodyIndex :type Integer) (bFromSweep :type Boolean) (SweepResult :type Hit%20Result%20Structure%20%28by%20ref%29))
  ;[0.1719 0.09859 0.1466] Trigger the RC Pawn for explosion when colliding with a Vehicle
  (Utilities|Casting|CastToFortVehicleInterface OtherActor)
  (Networking|SwitchHasAuthority
    (:Authority
      (if (not (RemoteControlledPawn|IsCurrentlyDelayingLaunch))
        (RemoteControlledPawn|TriggerDestroyRemoteControlPawn)))))

(event WaterBody|EventOnBeginWaterInteraction
  (params (WaterSurfaceLocation :type Vector :default "0, 0, 0") (WaterDepth :type Float%20%28single-precision%29))
  ;[0.04167 0.009739 0.04013] OnBeginWaterInteraction: Spawn water explosion effect and destroy actor
  (|CallParentFunction WaterSurfaceLocation WaterDepth)
  (Niagara|SpawnSystematLocation :SystemTemplate (Variables|FX|GetNS_WaterExplosion) :Location (Transformation|GetActorLocation) :Scale "1.000000,1.000000,1.000000" :bAutoDestroy true :bAutoActivate true :PoolingMethod "AutoRelease" :bPreCullCheck true)
  (RemoteControlledPawn|TriggerDestroyRemoteControlPawn))

(event EventEndPlay
  (params (EndPlayReason :type EEndPlayReason%20Enum :default "Destroyed"))
  ;[0.15 0.09514 0 0.5] EndPlay: Call killed if it hasn't been called yet.
  (Utilities|FlowControl|Sequence
    (:then_0
      (if (not (Variables|Transient|GetbKilled))
        (Default|CallOnKilled self)))
    (:then_1
      (|CallParentFunction EndPlayReason))))

(fn OnRep_bControlledUnpossessed
  (Variables|/Script/Engine.CharacterMovementComponent|SetbRunPhysicsWithNoController true (Variables|Character|GetCharacterMovement)))

(fn OnRep_bKilled
  (if (Variables|Transient|GetbKilled)
    (Default|CallOnKilled self)
    (Collision|SetActorEnableCollision false)
    (Networking|CanExecuteCosmeticEvents
      (:True
        (CallFunction|PlayExplosionEffects)))))

(fn ConstructionScript)
