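; Camera modifier for the ToyFox seeking missile explosion that pulls the view back and locks its rotation.
; On enable, it starts from a stored original location.
; Each frame it moves backward against the view's forward vector at a scalable-float speed until reaching a scalable-float distance from that origin, leaving FOV unchanged.
; View rotation processing zeroes the delta rotation and returns true, so the camera cannot turn.
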(event EventBlueprintModifierEnabled
  (Variables|Default|SetCurrentLoc (Variables|Default|GetOriginalLoc)))

(fn BlueprintModifyCamera
  (params (DeltaTime :type Float%20%28single-precision%29) (ViewLocation :type Vector :default "0, 0, 0") (ViewRotation :type Rotator :default "0, 0, 0") (FOV :type Float%20%28single-precision%29))
  (Utilities|FlowControl|Sequence
    (:then_0
      ; Move camera outward along the view rotation direction at given speed
      (if (< (Math|Vector|DistanceSquared(Vector) (Variables|Default|GetCurrentLoc) (Variables|Default|GetOriginalLoc)) (Math|Float|Square (ScalableFloat|GetCachedScalableFloatValue (Variables|Default|GetHF_DistanceToExplodeOriginalLoc))))
        (Variables|Default|SetCurrentLoc (- (Variables|Default|GetCurrentLoc) (Utilities|Operators|Multiply (Math|Rotator|GetRotationXVector (Variables|GetViewRotation)) (Variables|GetDeltaTime) (ScalableFloat|GetCachedScalableFloatValue (Variables|Default|GetHF_CamMoveSpeed)))))))
    (:then_1
      (return (Variables|Default|GetCurrentLoc) ViewRotation FOV))))

(fn BlueprintProcessViewRotation
  (params (ViewTarget :type Actor%20Object%20Reference) (DeltaTime :type Float%20%28single-precision%29) (InViewRotation :type Rotator :default "0, 0, 0") (InDeltaRot :type Rotator :default "0, 0, 0"))
  ; Stop camera rotation totally
  (return InViewRotation "0, 0, 0" true))
